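# Physics asset for the UpbeatIguanaMaple medium female head, giving its hair bones capsule bodies and joints for secondary motion.
# Neck, upper spine and clavicle capsules form a locked torso frame, with an empty upperarm_r body attached.
# Hair branches from the neck via base_hair_m and a free-jointed left master, which carries five locked base anchors.
# Dynamic strands have limited swing, mostly zero-degree twist, except free dyn_hair_s; strand c has two links.
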PhysicsAsset /BRCosmetics/Characters/Player/Female/Medium/Heads/F_MED_UpbeatIguanaMaple_Head/Meshes/CustomSkeleton/F_MED_UpbeatIguanaMaple_Head_Physics
bodies {
 base_hairMaster_l [{shapeType=Capsule, length=0.505, radius=0.505, shapeName=base_hairMaster_l_capsule, rotation=<rot 0 -90 0>}]
 base_hair_c [{shapeType=Capsule, length=0.505, radius=0.505, shapeName=base_hair_c_capsule, rotation=<rot 0 -90 0>}]
 base_hair_d [{shapeType=Capsule, length=0.505, radius=0.505, shapeName=base_hair_d_capsule, rotation=<rot 0 -90 0>}]
 base_hair_f [{shapeType=Capsule, length=0.505, radius=0.505, shapeName=base_hair_f_capsule, rotation=<rot 0 -90 0>}]
 base_hair_j [{shapeType=Capsule, length=0.505, radius=0.505, shapeName=base_hair_j_capsule, rotation=<rot 0 -90 0>}]
 base_hair_k [{shapeType=Capsule, length=0.505, radius=0.505, shapeName=base_hair_k_capsule, rotation=<rot 0 -90 0>}]
 base_hair_m [{shapeType=Capsule, length=0.505, radius=0.505, shapeName=base_hair_m_capsule, rotation=<rot 0 -90 0>}]
 clavicle_l [{
   shapeType=Capsule
   length=11.0241
   radius=6.3286
   shapeName=clavicle_l_capsule
   center=<4.2941 -1.3175 -0.3292>
   rotation=<rot 78.6004 -0.0002 -0.0002>
  }]
 clavicle_r [{
   shapeType=Capsule
   length=11.0241
   radius=6.3286
   shapeName=clavicle_l_capsule
   center=<-4.3202 1.3175 0.1996>
   rotation=<rot 78.6004 -0.0002 179.9998>
  }]
 dyn_hairMaster_l [{shapeType=Capsule, length=5.2594, radius=3.4133, shapeName=dyn_hairMaster_l_capsule, center=<6.5565 0 0>, rotation=<rot 0 -180 90>}]
 dyn_hair_01c [{shapeType=Capsule, length=2.7, radius=1.684, shapeName=dyn_hair_01c_capsule, center=<2.3681 0 0>, rotation=<rot 90 0 0>}]
 dyn_hair_02c [{
   shapeType=Capsule
   length=2.8276
   radius=1.7593
   shapeName=dyn_hair_02c_capsule
   center=<3.0346 0.2489 0.0224>
   rotation=<rot 74.5109 179.9999 174.4732>
  }]
 dyn_hair_d [{shapeType=Capsule, length=2.9232, radius=1.851, shapeName=dyn_hair_d_capsule, center=<2.847 0 -0.9126>, rotation=<rot 88.6277 0.0011 0.0011>}]
 dyn_hair_f [{shapeType=Capsule, length=3.3556, radius=1.8128, shapeName=dyn_hair_f_capsule, center=<2.5092 -0.185 0.1173>, rotation=<rot 90 0 0>}]
 dyn_hair_j [{
   shapeType=Capsule
   length=4.7676
   radius=1.899
   shapeName=dyn_hair_j_capsule
   center=<2.8304 0.3486 0.115>
   rotation=<rot 71.2315 -175.1985 -176.5224>
  }]
 dyn_hair_k [{shapeType=Capsule, length=3.276, radius=2.187, shapeName=dyn_hair_k_capsule, center=<0.8967 0 1.3342>, rotation=<rot 40.976 179.9973 180>}]
 dyn_hair_m [{shapeType=Capsule, length=8.4767, radius=3.3813, shapeName=dyn_hair_m_capsule, center=<4.6023 0 0.207>, rotation=<rot 0 0 90>}]
 dyn_hair_s [{
   shapeType=Capsule
   length=2.8874
   radius=1.9828
   shapeName=dyn_hair_s_capsule
   center=<2.2369 0.1068 1.5053>
   rotation=<rot 0 93.6493 -61.3836>
  }]
 neck_01 [{shapeType=Capsule, length=0.505, radius=0.505, shapeName=neck_01_capsule, rotation=<rot 90 0 0>}]
 spine_05 [{shapeType=Capsule, length=13.783, radius=9.5394, shapeName=spine_05_capsule, center=<1.4708 0 0>}]
 upperarm_r []
}
constraints [
 {bone1Name=dyn_hair_m, bone2Name=base_hair_m, swing1LimitDegrees=0, swing2LimitDegrees=30, twistLimitDegrees=20}
 {
  bone1Name=base_hair_m
  bone2Name=neck_01
  swing1LimitDegrees=45
  swing1Motion=Locked
  swing2LimitDegrees=45
  swing2Motion=Locked
  twistLimitDegrees=45
  twistMotion=Locked
 }
 {
  bone1Name=neck_01
  bone2Name=spine_05
  swing1LimitDegrees=45
  swing1Motion=Locked
  swing2LimitDegrees=45
  swing2Motion=Locked
  twistLimitDegrees=45
  twistMotion=Locked
 }
 {
  bone1Name=clavicle_r
  bone2Name=spine_05
  swing1LimitDegrees=45
  swing1Motion=Locked
  swing2LimitDegrees=45
  swing2Motion=Locked
  twistLimitDegrees=45
  twistMotion=Locked
 }
 {
  bone1Name=clavicle_l
  bone2Name=spine_05
  swing1LimitDegrees=45
  swing1Motion=Locked
  swing2LimitDegrees=45
  swing2Motion=Locked
  twistLimitDegrees=45
  twistMotion=Locked
 }
 {
  bone1Name=base_hair_c
  bone2Name=dyn_hairMaster_l
  swing1LimitDegrees=45
  swing1Motion=Locked
  swing2LimitDegrees=45
  swing2Motion=Locked
  twistLimitDegrees=45
  twistMotion=Locked
 }
 {bone1Name=dyn_hair_01c, bone2Name=base_hair_c, swing1LimitDegrees=45, swing2LimitDegrees=45, twistLimitDegrees=0}
 {bone1Name=dyn_hair_02c, bone2Name=dyn_hair_01c, swing1LimitDegrees=45, swing2LimitDegrees=45, twistLimitDegrees=0}
 {
  bone1Name=base_hair_j
  bone2Name=dyn_hairMaster_l
  swing1LimitDegrees=45
  swing1Motion=Locked
  swing2LimitDegrees=45
  swing2Motion=Locked
  twistLimitDegrees=45
  twistMotion=Locked
 }
 {bone1Name=dyn_hair_j, bone2Name=base_hair_j, swing1LimitDegrees=45, swing2LimitDegrees=45, twistLimitDegrees=0}
 {
  bone1Name=base_hair_k
  bone2Name=dyn_hairMaster_l
  swing1LimitDegrees=45
  swing1Motion=Locked
  swing2LimitDegrees=45
  swing2Motion=Locked
  twistLimitDegrees=45
  twistMotion=Locked
 }
 {bone1Name=dyn_hair_k, bone2Name=base_hair_k, swing1LimitDegrees=30, swing2LimitDegrees=45, twistLimitDegrees=0}
 {
  bone1Name=upperarm_r
  bone2Name=clavicle_r
  swing1LimitDegrees=45
  swing1Motion=Free
  swing2LimitDegrees=45
  swing2Motion=Free
  twistLimitDegrees=45
  twistMotion=Free
 }
 {
  bone1Name=base_hair_d
  bone2Name=dyn_hairMaster_l
  swing1LimitDegrees=45
  swing1Motion=Locked
  swing2LimitDegrees=45
  swing2Motion=Locked
  twistLimitDegrees=45
  twistMotion=Locked
 }
 {bone1Name=dyn_hair_d, bone2Name=base_hair_d, swing1LimitDegrees=30, swing2LimitDegrees=30, twistLimitDegrees=0}
 {
  bone1Name=base_hair_f
  bone2Name=dyn_hairMaster_l
  swing1LimitDegrees=45
  swing1Motion=Locked
  swing2LimitDegrees=45
  swing2Motion=Locked
  twistLimitDegrees=45
  twistMotion=Locked
 }
 {bone1Name=dyn_hair_f, bone2Name=base_hair_f, swing1LimitDegrees=40, swing2LimitDegrees=40, twistLimitDegrees=0}
 {
  bone1Name=base_hairMaster_l
  bone2Name=neck_01
  swing1LimitDegrees=45
  swing1Motion=Free
  swing2LimitDegrees=45
  swing2Motion=Free
  twistLimitDegrees=45
  twistMotion=Free
 }
 {
  bone1Name=dyn_hairMaster_l
  bone2Name=base_hairMaster_l
  swing1LimitDegrees=45
  swing1Motion=Free
  swing2LimitDegrees=45
  swing2Motion=Free
  twistLimitDegrees=45
  twistMotion=Free
 }
 {
  bone1Name=dyn_hair_s
  bone2Name=dyn_hair_m
  swing1LimitDegrees=45
  swing1Motion=Free
  swing2LimitDegrees=45
  swing2Motion=Free
  twistLimitDegrees=45
  twistMotion=Free
 }
]
