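# Post-process style AnimBlueprint that drives hair dynamics for the female UpbeatIguanaMaple head cosmetic.
# AnimGraph runs a Control Rig fed by RunningVar, then RigidBody and chained per-bone AnimDynamics, finishing with additive hair-bone offsets.
# EventGraph derives RunningVar from sprinting and lowers RigidBody alpha for tactical sprint, skydiving, targeting or forward jumps and falls.
# It also computes zipline, idle and gravity-override values, which the AnimGraph never reads.
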AnimBlueprint /BRCosmetics/Characters/Player/Female/Medium/Heads/F_MED_UpbeatIguanaMaple_Head/Meshes/CustomSkeleton/F_MED_Head_UpbeatIguanaMaple_AnimB
target_skeleton=&/Fortnite_Base_Head_UpBeatIguanaMap_Ske
graphs [
 AnimGraph {nodes [
   n1 = "Transform (Modify) Bone - Bone: dyn_hair_d"!(
    ComponentPose=Local To Component(LocalPose=Use cached pose 'CRCache_Dynamics'())
    bAlphaBoolEnabled:bool=True
    Translation="-0.722,-2.367,-0.000"
    Rotation="-0.000,-9.042,-0.000"
    Scale="1,1,1"
   ) -> (Pose) {
    alphaInputType=Bool
    rotationMode=BMM_Additive
    rotationSpace=BCS_BoneSpace
    translationMode=BMM_Additive
    translationSpace=BCS_BoneSpace
    alphaBoolBlend {blendInTime=0.3, blendOption=Linear, blendOutTime=0.3}
    boneToModify {boneName=dyn_hair_d}
    rotation {yaw=-9.042}
    translation=<-0.722 -2.367>
   }
   n2 = "Transform (Modify) Bone - Bone: dyn_hair_02c"!(ComponentPose=@n1, bAlphaBoolEnabled:bool=True, Translation="-2.064,-1.957,0.537", Rotation="-28.167,-27.314,1.746", Scale="1,1,1") -> (Pose) {
    alphaInputType=Bool
    rotationMode=BMM_Additive
    rotationSpace=BCS_BoneSpace
    translationMode=BMM_Additive
    translationSpace=BCS_BoneSpace
    alphaBoolBlend {blendInTime=0.3, blendOption=Linear, blendOutTime=0.3}
    boneToModify {boneName=dyn_hair_02c}
    rotation=<rot -28.167 -27.314 1.746>
    translation=<-2.064 -1.957 0.537>
   }
   n3 = "Transform (Modify) Bone - Bone: dyn_hair_f"!(ComponentPose=@n2, bAlphaBoolEnabled:bool=True, Translation="-0.116,0.197,1.104", Rotation="-1.253,7.840,10.786", Scale="1,1,1") -> (Pose) {
    alphaInputType=Bool
    rotationMode=BMM_Additive
    rotationSpace=BCS_BoneSpace
    translationMode=BMM_Additive
    translationSpace=BCS_BoneSpace
    alphaBoolBlend {blendInTime=0.3, blendOption=Linear, blendOutTime=0.3}
    boneToModify {boneName=dyn_hair_f}
    rotation=<rot -1.253 7.84 10.786>
    translation=<-0.116 0.1969 1.1036>
   }
   n4 = "Transform (Modify) Bone - Bone: dyn_hair_s"!(ComponentPose=@n3, bAlphaBoolEnabled:bool=True, Translation="-0.5,0,0", Rotation="15,0,0", Scale="1,1,1") -> (Pose) {
    alphaInputType=Bool
    rotationMode=BMM_Additive
    rotationSpace=BCS_BoneSpace
    translationMode=BMM_Additive
    translationSpace=BCS_BoneSpace
    alphaBoolBlend {blendInTime=0.2, blendOption=Linear, blendOutTime=0.2}
    boneToModify {boneName=dyn_hair_s}
    rotation {pitch=15}
    translation {x=-0.5}
   }
   n5 = "Transform (Modify) Bone - Bone: dyn_hair_k"!(ComponentPose=@n4, bAlphaBoolEnabled:bool=True, Translation="0.065,0.000,-0.670", Rotation="58.370,-0.002,0.001", Scale="1,1,1") -> (Pose) {
    alphaInputType=Bool
    rotationMode=BMM_Additive
    rotationSpace=BCS_BoneSpace
    translationMode=BMM_Additive
    translationSpace=BCS_BoneSpace
    alphaBoolBlend {blendInTime=0.2, blendOption=Linear, blendOutTime=0.2}
    boneToModify {boneName=dyn_hair_k}
    rotation=<rot 58.37 -0.002 0.001>
    translation {x=0.065, z=-0.67}
   }
   n6 = "Transform (Modify) Bone - Bone: dyn_hair_d"(ComponentPose=@n5, bAlphaBoolEnabled=True, Translation="-1.4,0.5,0", Rotation="-0.000,-42.863,0.000", Scale="1,1,1") {
    alphaInputType=Bool
    rotationMode=BMM_Additive
    rotationSpace=BCS_BoneSpace
    translationMode=BMM_Additive
    translationSpace=BCS_BoneSpace
    alphaBoolBlend {blendInTime=0.2, blendOption=Linear, blendOutTime=0.2}
    boneToModify {boneName=dyn_hair_d}
    rotation=<rot 0 -42.863 0>
    translation=<-1.4 0.5>
   }
   n7 = "Transform (Modify) Bone - Bone: dyn_hair_d"(ComponentPose=@n6, bAlphaBoolEnabled=True, Translation="0,0,0", Rotation="15,0,0", Scale="1,1,1") {
    alphaInputType=Bool
    rotationMode=BMM_Additive
    rotationSpace=BCS_BoneSpace
    translationMode=BMM_Additive
    translationSpace=BCS_BoneSpace
    alphaBoolBlend {blendInTime=0.2, blendOption=Linear, blendOutTime=0.2}
    boneToModify {boneName=dyn_hair_d}
    rotation {pitch=15}
   }
   Output Pose(Result=Component To Local(ComponentPose=@n7))
   n8 = Control Rig!(Source=Input Pose(), Alpha, RunVarCtrl:double=Get RunningVar() -> (RunningVar:double)) -> (Pose) {
    alphaBoolBlend {blendOption=Linear}
    controlRigAssetReference {blueprintRigClass=&/F_MED_UpbeatIguanaMaple_Head_CtrlRig.F_MED_UpbeatIguanaMaple_Head_CtrlRig_C}
   }
   Save cached pose 'CRCache_Input'(Pose=@n8) {cacheName=CRCache_Input}
   n9 = RigidBody!(
    ComponentPose=Local To Component(LocalPose=Use cached pose 'CRCache_Input'())
    Alpha=Get dynamicsAlpha() -> (dynamicsAlpha:double)
    OverrideWorldGravity:Vector="0,0,0"
    ExternalForce
   ) -> (Pose) {
    bOverrideWorldGravity=true
    overridePhysicsAsset=&/F_MED_UpbeatIguanaMaple_Head_Physics
    simulationSpace=BaseBoneSpace
    alphaBoolBlend {blendOption=Linear}
    baseBoneRef {boneName=spine_05}
    componentAppliedLinearAccClamp=<3000 3000 3000>
    componentLinearAccScale=<0.3 0.3 0.3>
    componentLinearVelScale=<0.3 0.3 0.3>
    simSpaceSettings {
     maxAngularAcceleration=1000
     maxAngularVelocity=3
     maxLinearAcceleration=700
     maxLinearVelocity=500
     velocityScaleZ=0
     worldAlpha=0.2
     externalLinearDragV=<0.2 0.2>
    }
   }
   n10 = Anim Dynamics - Bone: dyn_hair_a!(ComponentPose=@n9, Alpha=1, GravityOverride:Vector="0,0,-980", ExternalForce="0,0,0") -> (Pose) {
    angularDampingOverride=1
    angularSpringConstant=20000
    bAngularSpring=true
    bOverrideAngularDamping=true
    bUseGravityOverride=true
    gravityScale=0.35
    simulationSpace=RootRelative
    alphaBoolBlend {blendInTime=0.2, blendOption=Linear, blendOutTime=0.2}
    alphaScaleBiasClamp {bInterpResult=true, interpSpeedDecreasing=5, interpSpeedIncreasing=5}
    boundBone {boneName=dyn_hair_a}
    componentLinearAccScale=<1 1 1>
    componentLinearVelScale=<1 1 1>
    gravityOverride {z=-980}
    physicsBodyDefinitions [{
      boundBone {boneName=dyn_hair_a}
      boxExtents=<30 40 40>
      constraintSetup {angularLimitsMax=<0 30 30>, angularLimitsMin=<0 -30 -30>, angularTarget=<1 0 0>}
      localJointOffset=<4 0 0>
     }]
    relativeSpaceBone {boneName=pelvis}
   }
   n11 = Anim Dynamics - Bone: dyn_hair_g!(ComponentPose=@n10, Alpha=1, GravityOverride:Vector="0,0,-980", ExternalForce="0,0,0") -> (Pose) {
    angularDampingOverride=1
    angularSpringConstant=20000
    bAngularSpring=true
    bOverrideAngularDamping=true
    bUseGravityOverride=true
    gravityScale=0.35
    simulationSpace=RootRelative
    alphaBoolBlend {blendInTime=0.2, blendOption=Linear, blendOutTime=0.2}
    alphaScaleBiasClamp {bInterpResult=true, interpSpeedDecreasing=5, interpSpeedIncreasing=5}
    boundBone {boneName=dyn_hair_g}
    componentLinearAccScale=<1 1 1>
    componentLinearVelScale=<1 1 1>
    gravityOverride {z=-980}
    physicsBodyDefinitions [{
      boundBone {boneName=dyn_hair_g}
      boxExtents=<30 40 40>
      constraintSetup {angularLimitsMax=<0 30 30>, angularLimitsMin=<0 -30 -30>, angularTarget=<1 0 0>}
      localJointOffset=<4 0 0>
     }]
   }
   n12 = Anim Dynamics - Bone: dyn_hair_b!(ComponentPose=@n11, Alpha=1, GravityOverride:Vector="0,0,-980", ExternalForce="0,0,0") -> (Pose) {
    angularDampingOverride=1
    angularSpringConstant=20000
    bAngularSpring=true
    bOverrideAngularDamping=true
    bUseGravityOverride=true
    gravityScale=0.35
    simulationSpace=RootRelative
    alphaBoolBlend {blendInTime=0.2, blendOption=Linear, blendOutTime=0.2}
    alphaScaleBiasClamp {bInterpResult=true, interpSpeedDecreasing=5, interpSpeedIncreasing=5}
    boundBone {boneName=dyn_hair_b}
    componentLinearAccScale=<1 1 1>
    componentLinearVelScale=<1 1 1>
    gravityOverride {z=-980}
    physicsBodyDefinitions [{
      boundBone {boneName=dyn_hair_b}
      boxExtents=<30 30 30>
      constraintSetup {angularLimitsMax=<0 30 30>, angularLimitsMin=<0 -30 -30>, angularTarget=<1 0 0>}
      localJointOffset=<4 0 0>
     }]
    relativeSpaceBone {boneName=pelvis}
   }
   n13 = Anim Dynamics - Bone: dyn_hair_i!(ComponentPose=@n12, Alpha=1, GravityOverride:Vector="0,0,-980", ExternalForce="0,0,0") -> (Pose) {
    angularDampingOverride=1
    angularSpringConstant=20000
    bAngularSpring=true
    bOverrideAngularDamping=true
    bUseGravityOverride=true
    gravityScale=0.35
    simulationSpace=RootRelative
    alphaBoolBlend {blendInTime=0.2, blendOption=Linear, blendOutTime=0.2}
    alphaScaleBiasClamp {bInterpResult=true, interpSpeedDecreasing=5, interpSpeedIncreasing=5}
    boundBone {boneName=dyn_hair_i}
    componentLinearAccScale=<1 1 1>
    componentLinearVelScale=<1 1 1>
    gravityOverride {z=-980}
    physicsBodyDefinitions [{
      boundBone {boneName=dyn_hair_i}
      boxExtents=<30 30 30>
      constraintSetup {angularLimitsMax=<0 20 35>, angularLimitsMin=<0 -35 -35>, angularTarget=<1 0 0>}
      localJointOffset=<4 0 0>
     }]
   }
   n14 = Anim Dynamics - Bone: dyn_hair_h!(ComponentPose=@n13, Alpha=1, GravityOverride:Vector="0,0,-980", ExternalForce="0,0,0") -> (Pose) {
    angularDampingOverride=1
    angularSpringConstant=20000
    bAngularSpring=true
    bOverrideAngularDamping=true
    bUseGravityOverride=true
    gravityScale=0.35
    simulationSpace=RootRelative
    alphaBoolBlend {blendInTime=0.2, blendOption=Linear, blendOutTime=0.2}
    alphaScaleBiasClamp {bInterpResult=true, interpSpeedDecreasing=5, interpSpeedIncreasing=5}
    boundBone {boneName=dyn_hair_h}
    componentLinearAccScale=<1 1 1>
    componentLinearVelScale=<1 1 1>
    gravityOverride {z=-980}
    physicsBodyDefinitions [{
      boundBone {boneName=dyn_hair_h}
      boxExtents=<30 30 30>
      constraintSetup {angularLimitsMax=<0 15 35>, angularLimitsMin=<0 -35 -35>, angularTarget=<1 0 0>}
      localJointOffset=<4 0 0>
     }]
    relativeSpaceBone {boneName=pelvis}
   }
   n15 = Anim Dynamics - Bone: dyn_hair_e!(ComponentPose=@n14, Alpha=1, GravityOverride:Vector="0,0,-980", ExternalForce="0,0,0") -> (Pose) {
    angularDampingOverride=0.9999
    angularSpringConstant=10000
    bAngularSpring=true
    bOverrideAngularDamping=true
    bUseGravityOverride=true
    gravityScale=0.35
    simulationSpace=BoneRelative
    alphaBoolBlend {blendInTime=0.2, blendOption=Linear, blendOutTime=0.2}
    alphaScaleBiasClamp {bInterpResult=true, interpSpeedDecreasing=5, interpSpeedIncreasing=5}
    boundBone {boneName=dyn_hair_e}
    componentLinearAccScale=<1 1 1>
    componentLinearVelScale=<1 1 1>
    gravityOverride {z=-980}
    physicsBodyDefinitions [{
      boundBone {boneName=dyn_hair_e}
      boxExtents=<20 30 30>
      constraintSetup {angularLimitsMax=<0 30 30>, angularLimitsMin=<0 -30 -30>, angularTarget=<1 0 0>}
      localJointOffset=<5 0 0>
     }]
    relativeSpaceBone {boneName=pelvis}
   }
   n16 = Anim Dynamics - Bone: dyn_hair_n!(ComponentPose=@n15, Alpha=1, GravityOverride:Vector="0,0,-980", ExternalForce="0,0,0") -> (Pose) {
    angularDampingOverride=1
    angularSpringConstant=10000
    bAngularSpring=true
    bOverrideAngularDamping=true
    bUseGravityOverride=true
    gravityScale=0.35
    alphaBoolBlend {blendInTime=0.2, blendOption=Linear, blendOutTime=0.2}
    alphaScaleBiasClamp {bInterpResult=true, interpSpeedDecreasing=5, interpSpeedIncreasing=5}
    boundBone {boneName=dyn_hair_n}
    componentLinearAccScale=<1 1 1>
    componentLinearVelScale=<1 1 1>
    gravityOverride {z=-980}
    physicsBodyDefinitions [{
      boundBone {boneName=dyn_hair_n}
      boxExtents=<20 30 30>
      constraintSetup {angularLimitsMax=<0 0 15>, angularLimitsMin=<0 -20 -15>, angularTarget=<1 0 0>}
      localJointOffset=<4 0 0>
     }]
   }
   n17 = Anim Dynamics - Bone: dyn_hair_w!(ComponentPose=@n16, Alpha=1, GravityOverride:Vector="0,0,-980", ExternalForce="0,0,0") -> (Pose) {
    angularDampingOverride=1
    angularSpringConstant=20000
    bAngularSpring=true
    bOverrideAngularDamping=true
    bUseGravityOverride=true
    gravityScale=0.35
    simulationSpace=RootRelative
    alphaBoolBlend {blendOption=Linear}
    boundBone {boneName=dyn_hair_w}
    componentLinearAccScale=<1 1 1>
    componentLinearVelScale=<1 1 1>
    gravityOverride {z=-980}
    physicsBodyDefinitions [{
      boundBone {boneName=dyn_hair_w}
      boxExtents=<30 30 30>
      constraintSetup {angularLimitsMax=<0 15 5>, angularLimitsMin=<0 -25 -20>, angularTarget=<1 0 0>}
      localJointOffset=<2 0 0>
     }]
   }
   n18 = Anim Dynamics - Bone: dyn_hair_o!(ComponentPose=@n17, Alpha=1, GravityOverride:Vector="0,0,-980", ExternalForce="0,0,0") -> (Pose) {
    angularDampingOverride=1
    angularSpringConstant=10000
    bAngularSpring=true
    bOverrideAngularDamping=true
    bUseGravityOverride=true
    gravityScale=0.35
    alphaBoolBlend {blendInTime=0.2, blendOption=Linear, blendOutTime=0.2}
    alphaScaleBiasClamp {bInterpResult=true, interpSpeedDecreasing=5, interpSpeedIncreasing=5}
    boundBone {boneName=dyn_hair_o}
    componentLinearAccScale=<1 1 1>
    componentLinearVelScale=<1 1 1>
    gravityOverride {z=-980}
    physicsBodyDefinitions [{
      boundBone {boneName=dyn_hair_o}
      boxExtents=<20 30 30>
      constraintSetup {angularLimitsMax=<0 0 15>, angularLimitsMin=<0 -20 -15>, angularTarget=<1 0 0>}
      localJointOffset=<4 0 0>
     }]
   }
   n19 = Anim Dynamics - Bone: dyn_hair_p!(ComponentPose=@n18, Alpha=1, GravityOverride:Vector="0,0,-980", ExternalForce="0,0,0") -> (Pose) {
    angularDampingOverride=1
    angularSpringConstant=10000
    bAngularSpring=true
    bOverrideAngularDamping=true
    bUseGravityOverride=true
    gravityScale=0.35
    alphaBoolBlend {blendInTime=0.2, blendOption=Linear, blendOutTime=0.2}
    alphaScaleBiasClamp {bInterpResult=true, interpSpeedDecreasing=5, interpSpeedIncreasing=5}
    boundBone {boneName=dyn_hair_p}
    componentLinearAccScale=<1 1 1>
    componentLinearVelScale=<1 1 1>
    gravityOverride {z=-980}
    physicsBodyDefinitions [{
      boundBone {boneName=dyn_hair_p}
      boxExtents=<20 30 30>
      constraintSetup {angularLimitsMax=<0 0 15>, angularLimitsMin=<0 -20 -15>, angularTarget=<1 0 0>}
      localJointOffset=<4 0 0>
     }]
   }
   n20 = Anim Dynamics - Bone: dyn_hair_q!(ComponentPose=@n19, Alpha=1, GravityOverride:Vector="0,0,-980", ExternalForce="0,0,0") -> (Pose) {
    angularDampingOverride=1
    angularSpringConstant=10000
    bAngularSpring=true
    bOverrideAngularDamping=true
    bUseGravityOverride=true
    gravityScale=0.35
    alphaBoolBlend {blendInTime=0.2, blendOption=Linear, blendOutTime=0.2}
    alphaScaleBiasClamp {bInterpResult=true, interpSpeedDecreasing=5, interpSpeedIncreasing=5}
    boundBone {boneName=dyn_hair_q}
    componentLinearAccScale=<1 1 1>
    componentLinearVelScale=<1 1 1>
    gravityOverride {z=-980}
    physicsBodyDefinitions [{
      boundBone {boneName=dyn_hair_q}
      boxExtents=<20 30 30>
      constraintSetup {angularLimitsMax=<0 0 15>, angularLimitsMin=<0 -20 -15>, angularTarget=<1 0 0>}
      localJointOffset=<4 0 0>
     }]
   }
   n21 = Anim Dynamics - Bone: dyn_hair_r!(ComponentPose=@n20, Alpha=1, GravityOverride:Vector="0,0,-980", ExternalForce="0,0,0") -> (Pose) {
    angularDampingOverride=1
    angularSpringConstant=10000
    bAngularSpring=true
    bOverrideAngularDamping=true
    bUseGravityOverride=true
    gravityScale=0.35
    alphaBoolBlend {blendInTime=0.2, blendOption=Linear, blendOutTime=0.2}
    alphaScaleBiasClamp {bInterpResult=true, interpSpeedDecreasing=5, interpSpeedIncreasing=5}
    boundBone {boneName=dyn_hair_r}
    componentLinearAccScale=<1 1 1>
    componentLinearVelScale=<1 1 1>
    gravityOverride {z=-980}
    physicsBodyDefinitions [{
      boundBone {boneName=dyn_hair_r}
      boxExtents=<20 30 30>
      constraintSetup {angularLimitsMax=<0 0 15>, angularLimitsMin=<0 -20 -15>, angularTarget=<1 0 0>}
      localJointOffset=<4 0 0>
     }]
   }
   n22 = Anim Dynamics - Bone: dyn_hair_s!(ComponentPose=@n21, Alpha=1, GravityOverride:Vector="0,0,-980", ExternalForce="0,0,0") -> (Pose) {
    angularDampingOverride=1
    angularSpringConstant=10000
    bAngularSpring=true
    bOverrideAngularDamping=true
    bUseGravityOverride=true
    gravityScale=0.35
    alphaBoolBlend {blendInTime=0.2, blendOption=Linear, blendOutTime=0.2}
    alphaScaleBiasClamp {bInterpResult=true, interpSpeedDecreasing=5, interpSpeedIncreasing=5}
    boundBone {boneName=dyn_hair_s}
    componentLinearAccScale=<1 1 1>
    componentLinearVelScale=<1 1 1>
    gravityOverride {z=-980}
    physicsBodyDefinitions [{
      boundBone {boneName=dyn_hair_s}
      boxExtents=<20 30 30>
      constraintSetup {angularLimitsMax=<0 0 15>, angularLimitsMin=<0 -15 -15>, angularTarget=<1 0 0>}
      localJointOffset=<4 0 0>
     }]
   }
   n23 = Anim Dynamics - Bone: dyn_hair_t!(ComponentPose=@n22, Alpha=1, GravityOverride:Vector="0,0,-980", ExternalForce="0,0,0") -> (Pose) {
    angularDampingOverride=1
    angularSpringConstant=10000
    bAngularSpring=true
    bOverrideAngularDamping=true
    bUseGravityOverride=true
    gravityScale=0.35
    alphaBoolBlend {blendInTime=0.2, blendOption=Linear, blendOutTime=0.2}
    alphaScaleBiasClamp {bInterpResult=true, interpSpeedDecreasing=5, interpSpeedIncreasing=5}
    boundBone {boneName=dyn_hair_t}
    componentLinearAccScale=<1 1 1>
    componentLinearVelScale=<1 1 1>
    gravityOverride {z=-980}
    physicsBodyDefinitions [{
      boundBone {boneName=dyn_hair_t}
      boxExtents=<20 30 30>
      constraintSetup {angularLimitsMax=<0 0 15>, angularLimitsMin=<0 -20 -15>, angularTarget=<1 0 0>}
      localJointOffset=<4 0 0>
     }]
   }
   n24 = Anim Dynamics - Bone: dyn_hair_u!(ComponentPose=@n23, Alpha=1, GravityOverride:Vector="0,0,-980", ExternalForce="0,0,0") -> (Pose) {
    angularDampingOverride=1
    angularSpringConstant=20000
    bAngularSpring=true
    bOverrideAngularDamping=true
    bUseGravityOverride=true
    gravityScale=0.35
    simulationSpace=RootRelative
    alphaBoolBlend {blendOption=Linear}
    boundBone {boneName=dyn_hair_u}
    componentLinearAccScale=<1 1 1>
    componentLinearVelScale=<1 1 1>
    gravityOverride {z=-980}
    physicsBodyDefinitions [{
      boundBone {boneName=dyn_hair_u}
      boxExtents=<30 30 30>
      constraintSetup {angularLimitsMax=<0 5 25>, angularLimitsMin=<0 -30 -25>, angularTarget=<1 0 0>}
      localJointOffset=<4 0 0>
     }]
   }
   n25 = Anim Dynamics - Bone: dyn_hair_v!(ComponentPose=@n24, Alpha=1, GravityOverride:Vector="0,0,-980", ExternalForce="0,0,0") -> (Pose) {
    angularDampingOverride=1
    angularSpringConstant=20000
    bAngularSpring=true
    bOverrideAngularDamping=true
    bUseGravityOverride=true
    gravityScale=0.35
    simulationSpace=RootRelative
    alphaBoolBlend {blendOption=Linear}
    boundBone {boneName=dyn_hair_v}
    componentLinearAccScale=<1 1 1>
    componentLinearVelScale=<1 1 1>
    gravityOverride {z=-980}
    physicsBodyDefinitions [{
      boundBone {boneName=dyn_hair_v}
      boxExtents=<30 30 30>
      constraintSetup {angularLimitsMax=<0 5 25>, angularLimitsMin=<0 -30 -25>, angularTarget=<1 0 0>}
      localJointOffset=<4 0 0>
     }]
   }
   n26 = Anim Dynamics - Bone: dyn_hairMaster_l!(ComponentPose=@n25, Alpha=1, GravityOverride:Vector="0,0,-980", ExternalForce="0,0,0") -> (Pose) {
    angularDampingOverride=1
    angularSpringConstant=20000
    bAngularSpring=true
    bOverrideAngularDamping=true
    bUseGravityOverride=true
    gravityScale=0.35
    alphaBoolBlend {blendOption=Linear}
    boundBone {boneName=dyn_hairMaster_l}
    componentLinearAccScale=<1 1 1>
    componentLinearVelScale=<1 1 1>
    gravityOverride {z=-980}
    physicsBodyDefinitions [{
      boundBone {boneName=dyn_hairMaster_l}
      boxExtents=<30 30 30>
      constraintSetup {angularLimitsMax=<15 1 18>, angularLimitsMin=<-15 -1 -18>, angularTarget=<1 0 0>}
      localJointOffset=<4 0 0>
     }]
   }
   n27 = Anim Dynamics - Bone: dyn_hairMaster02_bk!(ComponentPose=@n26, Alpha=1, GravityOverride:Vector="0,0,-980", ExternalForce="0,0,0") -> (Pose) {
    angularDampingOverride=1
    angularSpringConstant=20000
    bAngularSpring=true
    bOverrideAngularDamping=true
    bUseGravityOverride=true
    gravityScale=0.35
    alphaBoolBlend {blendOption=Linear}
    boundBone {boneName=dyn_hairMaster02_bk}
    componentLinearAccScale=<1 1 1>
    componentLinearVelScale=<1 1 1>
    gravityOverride {z=-980}
    physicsBodyDefinitions [{
      boundBone {boneName=dyn_hairMaster02_bk}
      boxExtents=<30 30 30>
      constraintSetup {angularLimitsMax=<0 10 8>, angularLimitsMin=<0 -15 -8>, angularTarget=<1 0 0>}
      localJointOffset=<4 0 0>
     }]
   }
   Save cached pose 'CRCache_Dynamics'(Pose=Component To Local(ComponentPose=@n27)) {cacheName=CRCache_Dynamics}
  ]}
 EventGraph {nodes [
   Event BlueprintUpdateAnimation()
   > n1 = Sequence()
   > Reroute Node()
   n2 = Sequence!(execute=@n1.then_1) -> (then_0, then_1, then_2:Exec, then_3:Exec)
   n3 = Gravity Override State Function DEPRECATED(
    execute=@n2.then_1
    CharacterPartAnimInstance=Self-Reference()
    GameTimeInSeconds=GetGameTimeInSeconds()
    GravityOverrideParams=Get Gravity Override Params()
   )
   > Set GravityOverride(GravityOverride=@n3)
   n4 = Get OnZipline() -> (OnZipline:bool)
   Set ZiplineAlpha(execute=@n2.then_3, ZiplineAlpha:double="To Float (Boolean)"(InBool=NOT Boolean(A=@n4))) -> (Output_Get:double)
   n5 = Get bIsMoving2D()
   n6 = Get bIsTargeting()
   > Set IdlePose(IdlePose:bool=AND Boolean!(A=NOT Boolean(A=@n5), B=NOT Boolean(A=@n6), C:bool=Get bRigidBodyEnabled()) -> (ReturnValue)) -> (Output_Get:bool)
   n7 = NOT Boolean(A=OR Boolean(A=Get bIsJumping(), B=Get bIsFalling()))
   n8 = Get bIsSprinting()
   > Set RunningVar(RunningVar:double="To Float (Boolean)"(InBool=AND Boolean!(A=@n8, B=@n7) -> (ReturnValue))) -> (Output_Get:double)
   n9 = Branch(execute=@n2.then_2, Condition=OR Boolean(A=Get bIsTacticalSprinting(), B=Get bIsSkyDiving()))
   > Set dynamicsAlpha(dynamicsAlpha:double=0.2) -> (Output_Get:double)
   Set dynamicsAlpha(execute=@n9.else, dynamicsAlpha=1)
   n10 = OR Boolean(A=Get bIsJumping(), B=Get bIsFalling())
   n11 = Get forwardBool() -> (forwardBool:bool)
   > n12 = Branch(Condition=OR Boolean(A=AND Boolean(A=@n11, B=@n10), B=Get bIsTargeting()))
   > Set dynamicsAlpha(dynamicsAlpha=0.7)
   Set dynamicsAlpha(execute=@n12.else, dynamicsAlpha=1)
   n13 = PlayerMovement(execute=@n2.then_0, AnimInstance=Self-Reference(), XDotProductCoefficient=100, YDotProductCoefficient=100, ZDotProductCoefficient=100)
   n14 = BreakVector(InVec=MakeVector(X=@n13, Y=@n13.YMovement))
   > Set RightBool(RightBool:bool=AND Boolean(A="float > float"(A=@n14, B=50), B=Get bIsParachuteOpened())) -> (Output_Get:bool)
   > Set forwardBool(forwardBool:bool="float > float"(A=@n14.Y, B=40)) -> (Output_Get:bool)
   n15 = Event BlueprintProcessAnimInputEvent()
   > n16 = Branch(Condition=IsNamed(self=@n15.AnimInputEvent, InEventName=aie.Player.Zipline.Entered))
   > Set OnZipline(OnZipline:bool=true) -> (Output_Get:bool)
   Branch(execute=@n16.else, Condition=IsNamed(self=@n15.AnimInputEvent, InEventName=aie.Player.Zipline.Exited))
   > Set OnZipline(OnZipline=false)
   TryGetPawnOwner()
  ]}
]
variables [
 dynamicsAlpha {type=number}
 forwardBool {type=boolean}
 Gravity Override Params {type=GravityOverrideParamsStruct}
 GravityOverride {type=Vector}
 IdlePose {type=boolean}
 LookUpHead {type=number}
 OnZipline {type=boolean}
 RightBool {type=boolean}
 RunningVar {type=number}
 Skel Mesh {type=SkeletalMeshComponent}
 ZiplineAlpha {type=number}
]
